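# Translucent unlit Niagara sprite material, named for digital cube rune effects on the Cloaked Assassin NPC back bling.
# Particle color RGB runs through the shared emissive adjustment function to provide the glow color.
# A dynamic parameter drives a random RGB-channel pick from the mask texture, so one texture can supply several shapes.
# The picked mask is sharpened by a high power and multiplied by particle alpha, apparently for opacity.
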Material /BRCosmetics/Accessories/FORT_Backpacks/Backpack_F_MED_Cloaked_Assassin_NPC/FX/M_Cube_DigiRunes_CloakedAssassian_NPC
graph {
 properties {bUsedWithNiagaraSprites=true, blendMode=BLEND_Translucent, pixelDepthOffsetMode=PDOM_Legacy, shadingModel=MSM_Unlit, usedShadingModels=Unlit}
 nodes [
  n1 = ParticleColor()
  MaterialFunctionCall(Emissive=@n1.RGB) {materialFunction=&MF_FX_EmissiveAdjustments}
  n2 = DynamicParameter() {defaultValue=<rgba 0.65 1 1 1>, paramNames [Random Selection, Param2, Param3, Param4]}
  n3 = TextureSampleParameter2D() {parameterName=Texture, samplerType=SAMPLERTYPE_LinearColor, texture=/Game/Athena/Prototype/Blueprints/Cube/Materials/DFHeightMask}
  Multiply(
   A=@n1.A
   B=Power(Base=MaterialFunctionCall!(Random Picker=@n2.Random Selection, Texture Input=@n3.RGB, Option 1, Option 2, Option 3) {materialFunction=/Game/Effects/Fort_Effects/Materials/_Functions/MF_RGB_Random_01}.Result) {constExponent=8}
  )
 ]
}
