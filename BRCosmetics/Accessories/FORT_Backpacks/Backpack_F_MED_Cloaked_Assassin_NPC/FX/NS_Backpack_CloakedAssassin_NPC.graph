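# Niagara system for the Cloaked Assassin NPC backpack cosmetic, tinting lightning, lights, runes and mesh debris from a User FXColor (magenta by default).
# Lightning sprites and pulsing lights follow sampled points on the User skeletal mesh, while four persistent lights track separate bones or sockets by spawn group.
# Polygon meshes spawned on the mesh and small rune sprites fly off along negative X, decelerating by velocity curves as they shrink or fade.
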NiagaraSystem /BRCosmetics/Accessories/FORT_Backpacks/Backpack_F_MED_Cloaked_Assassin_NPC/FX/NS_Backpack_CloakedAssassin_NPC
dependencies {
 usedDataInterfaces [
  &NiagaraDataInterfaceCurve
  &NiagaraDataInterfaceColorCurve
  &NiagaraDataInterfaceArrayFloat
  &NiagaraDataInterfaceMeshRendererInfo
 ]
 usedDynamicInputs []
 usedModules [&SystemState]
}
emitters {
 Lightning {
  emitterUpdateScript [
   &EmitterState()
   &SpawnRate(SpawnRate=10, Use Spawn Probability=true)
  ]
  particleSpawnScript [
   &InitializeParticle(
    AdjustAlpha=false
    AdjustHue=true
    AdjustSaturation=false
    AdjustValue=true
    Color=$User.FXColor
    Color Mode=::NewEnumerator3
    Hue Shift Range:Vector2f=<0 0.1>
    Lifetime Max=0.4
    Lifetime Min=0.2
    Lifetime Mode=::NewEnumerator1
    Sprite Rotation Mode=::NewEnumerator2
    Sprite Size Mode=::NewEnumerator1
    Uniform Sprite Size Max=17
    Uniform Sprite Size Min=6
    UsePositionOffset=false
    Value Range:Vector2f=<1 2>
    Write Material Random=true
   ) {module=InitializeParticle001}
   &DynamicMaterialParameters(
    Index 0 Param 1=&UniformRangedFloat(Minimum=0, Maximum=1)
    Index 0 Param 2=&UniformRangedFloat(Minimum=0.1, Maximum=1)
    Index 0 Param 3=1
    Module.Param0WriteEnabled=true
    Module.Param1WriteEnabled=true
    Module.Param2WriteEnabled=true
    Module.Param3WriteEnabled=false
   ) {module=DynamicMaterialParameters001}
   /SetVariables(Particles.SubImageIndex=&UniformAOrBFloat(A=1, B=3)) {isSetParameters=true}
  ]
  particleUpdateScript [
   &ParticleState()
   &SkeletalMeshLocation(
    Inherit Velocity=false
    MaskSpawnGroup=false
    Mesh Sampling Type=::NewEnumerator5
    Sampled Position Offset=<-5 0 0>
    Sampled Position Offset Coordinate Space:ENiagaraMeshTransforms=::NewEnumerator2
    Skeletal Mesh=$User.SkeletalMesh
    Socket Sampling Mode=::NewEnumerator5
   )
   &DynamicMaterialParameters(
    Index 0 Param 3=&FloatFromCurve(FloatCurve=DataInterface() {propertyValues {Curve {keys [<at 0 1 RCIM_Cubic>,<at 0.2 0.025 RCIM_Cubic>,<at 1 1>]}}})
    Module.Param0WriteEnabled=false
    Module.Param1WriteEnabled=false
    Module.Param2WriteEnabled=true
    Module.Param3WriteEnabled=false
   )
   &ScaleColor(
    Scale RGB=&Multiply_VectorByFloat(
     Vector=&MakeVectorFromLinearColorRGB(Color=&ColorFromCurve(ColorCurve=DataInterface() {propertyValues {
         AlphaCurve {keys []}
         BlueCurve {keys [<at 0 1>, {arriveTangent=-0.4, interpMode=RCIM_Cubic, leaveTangent=-0.4, time=0.15, value=10}, <at 1 0.6 RCIM_Cubic>]}
         GreenCurve {keys [<at 0 1>, {arriveTangent=-0.4, interpMode=RCIM_Cubic, leaveTangent=-0.4, time=0.15, value=10}, <at 1 0.6 RCIM_Cubic>]}
         RedCurve {keys [<at 0 1>, {arriveTangent=-0.4, interpMode=RCIM_Cubic, leaveTangent=-0.4, time=0.15, value=10}, <at 1 0.6 RCIM_Cubic>]}
        }}))
     Float=1
    )
    ScaleA=false
    ScaleRGB=true
   )
   &SubUVAnimation(
    End Frame=5
    SubUV Lookup Index:NiagaraFloat=&Add_Float(A=&Multiply_Float(A=0.5, B=$Particles.NormalizedAge), B=$Particles.Initial.SubImageIndex)
    SubUV Loop Count=0.5
   ) {module=SubUVAnimation001}
  ]
  rendererClasses [&NiagaraSpriteRendererProperties]
  renderers [{rendererClass=&NiagaraSpriteRendererProperties}]
 }
 Lights {
  emitterUpdateScript [
   &EmitterState()
   &SpawnRate(SpawnRate=2, Use Spawn Probability=false)
  ]
  particleSpawnScript [
   &InitializeParticle(Color=$User.FXColor, Lifetime Min=0.3, Lifetime Mode=::NewEnumerator1, UsePositionOffset=false, Write Material Random=true) {module=InitializeParticle001}
   /SetVariables(Particles.LightExponent=1) {isSetParameters=true, module=SetVariables_#1}
   /SetVariables(Particles.LightRadius=15) {isSetParameters=true, module=SetVariables_#2}
  ]
  particleUpdateScript [
   &ParticleState()
   &SkeletalMeshLocation(
    Inherit Velocity=false
    MaskSpawnGroup=false
    Mesh Sampling Type=::NewEnumerator5
    Sampled Position Offset=<-7 0 0>
    Sampled Position Offset Coordinate Space:ENiagaraMeshTransforms=::NewEnumerator2
    Skeletal Mesh=$User.SkeletalMesh
    Socket Sampling Mode=::NewEnumerator5
   )
   &ScaleColor(
    Scale RGB=&Multiply_VectorByFloat(
     Vector=<50 50 50>
     Float=&FloatFromCurve(FloatCurve=DataInterface() {propertyValues {Curve {keys [
          {arriveTangent=4.8231, interpMode=RCIM_Cubic, leaveTangent=4.8231, tangentMode=RCTM_User, time=0, value=0}
          {interpMode=RCIM_Cubic, tangentMode=RCTM_User, time=0.3294, value=1}
          {arriveTangent=-1.3928, interpMode=RCIM_Cubic, leaveTangent=-1.3928, tangentMode=RCTM_User, time=1, value=0}
         ]}}})
    )
    ScaleA=false
    ScaleRGB=true
   )
  ]
  rendererClasses [&NiagaraLightRendererProperties]
  renderers [{rendererClass=&NiagaraLightRendererProperties, rendererIndex=0}]
 }
 Lights_BackpackArms {
  emitterUpdateScript [
   &EmitterState(Life Cycle Mode=::NewEnumerator1, Loop Behavior=::NewEnumerator1, Loop Duration Mode=::NewEnumerator1, UseLoopDelay=false)
   &SpawnBurst_Instantaneous(Spawn Count=4, Use Spawn Probability=false, UseLoopCountLimit=false)
  ]
  particleSpawnScript [
   &FN_SetSpawnGroup()
   &InitializeParticle(Color=$User.FXColor, Lifetime=0, UsePositionOffset=false, Write Material Random=true) {module=InitializeParticle001}
   /SetVariables(Particles.LightExponent=3) {isSetParameters=true, module=SetVariables_#1}
   /SetVariables(Particles.LightRadius=&SelectFloatFromArray(Direct Array Index=$Particles.SpawnGroup, Float Selection Array=DataInterface() {propertyValues {FloatData [15,15,10,10]}})) {isSetParameters=true, module=SetVariables_#2}
  ]
  particleUpdateScript [
   &ParticleState(Kill Particles When Lifetime Has Elapsed=false, Let Infinitely Lived Particles Die When Emitter Deactivates=true)
   &SkeletalMeshLocation(
    Bone / Socket Index=&Add_Integer(A=$Particles.SpawnGroup, B=1)
    Inherit Velocity=false
    MaskSpawnGroup=false
    Mesh Sampling Type=::NewEnumerator5
    Sampled Position Offset Coordinate Space:ENiagaraMeshTransforms=::NewEnumerator2
    Skeletal Mesh=$User.SkeletalMesh
    Socket Sampling Mode=::NewEnumerator5
   )
   &ScaleColor(Scale RGB=&Multiply_VectorByFloat(Vector=<50 50 50>, Float=1), ScaleA=false, ScaleRGB=true)
  ]
  rendererClasses [&NiagaraLightRendererProperties]
  renderers [{rendererClass=&NiagaraLightRendererProperties, rendererIndex=0}]
 }
 Polygon_Trail {
  emitterUpdateScript [
   &EmitterState()
   &SpawnRate(Spawn Probability=0.2, SpawnRate=20, Use Spawn Probability=true)
  ]
  particleSpawnScript [
   &InitializeParticle(
    Color=$User.FXColor
    Lifetime Max=0.7
    Lifetime Min=0.2
    Lifetime Mode=::NewEnumerator1
    Mesh Scale Mode=::NewEnumerator1
    Mesh Uniform Scale Max=0.05
    Mesh Uniform Scale Min=0.02
    UsePositionOffset=false
    Write Material Random=true
   )
   /SetVariables(Particles.MeshIndex=&RandomRangeInt(Minimum=0, Maximum=1)) {isSetParameters=true}
   &SkeletalMeshLocation(MaskSpawnGroup=false, OverrideRandomSeed=false, Skeletal Mesh=$User.SkeletalMesh)
   &InitialMeshOrientation(Use Rotation Vector=false)
   /SetVariables(Particles.ColorB:LinearColor=&MultiplyLinearColor(
     Linear Color=$User.FXColor
     Scale Factor=&MakeFloat4FromVectorAndFloat("Vector (RGB)"=&VectorFromFloat(Value=0.1), "Float (Alpha)"=1)
    )) {isSetParameters=true}
   &AddVelocity(Velocity=&RandomRangeVector(Minimum=<-300 -200 -200>, Maximum=<-600 200 200>)) {module=AddVelocity001}
  ]
  particleUpdateScript [
   &ParticleState()
   /SetVariables(Particles.Velocity=&Multiply_VectorByFloat(
     Vector=$Particles.Initial.Velocity
     Float=&FloatFromCurve(FloatCurve=DataInterface() {propertyValues {Curve {keys [
          {arriveTangent=-5.5773, interpMode=RCIM_Cubic, leaveTangent=-5.5773, tangentMode=RCTM_User, time=0, value=1}
          {arriveTangent=-0.9831, interpMode=RCIM_Cubic, leaveTangent=-0.9831, tangentMode=RCTM_User, time=0.2018, value=0.1945}
          <at 0.75 0>
         ]}}})
    )) {isSetParameters=true, module=SetVariables_#3}
   &ScaleMeshSize(Scale Factor=&VectorFromFloat(Value=&FloatFromCurve(FloatCurve=DataInterface() {propertyValues {Curve {keys [<at 0 1>,<at 1 0>]}}})))
   &SolveForcesAndVelocity(Clamp Velocity=false, Limit Acceleration=false)
  ]
  rendererClasses [&NiagaraMeshRendererProperties]
  renderers [
   {rendererClass=&NiagaraMeshRendererProperties, rendererIndex=0}
   {rendererClass=&NiagaraMeshRendererProperties, rendererIndex=1}
  ]
 }
 Runes {
  emitterUpdateScript [
   &EmitterState()
   &SpawnRate(Spawn Probability=0.2, SpawnRate=20, Use Spawn Probability=true)
  ]
  particleSpawnScript [
   &InitializeParticle(
    AdjustAlpha=false
    AdjustHue=false
    AdjustSaturation=false
    AdjustValue=true
    Color=$User.FXColor
    Color Mode=::NewEnumerator3
    Lifetime Max=1
    Lifetime Min=0.1
    Lifetime Mode=::NewEnumerator1
    Sprite Rotation Mode=::NewEnumerator2
    Sprite Size Mode=::NewEnumerator1
    Uniform Sprite Size Max=6
    Uniform Sprite Size Min=4
    UsePositionOffset=false
    Value Range:Vector2f=<3 5>
    Write Material Random=true
   ) {module=InitializeParticle001}
   &SphereLocation(
    Hemisphere X=true
    MaskSpawnGroup=false
    Override Seed=false
    Sphere Radius=3
    Surface Only=false
    UseNonUniformScale=false
    UseOffset=false
    UseSphereOrientation=false
   )
   &AddVelocity(Velocity=&RandomRangeVector(Minimum=<-150 -100 -100>, Maximum=<-300 100 100>))
   &SubUVAnimation(End Frame=3, SubUV Animation Mode:ENiagara_SubUVLookupMode=::NewEnumerator2) {module=SubUVAnimation001}
   &DynamicMaterialParameters(
    Index 0 Param 1=&RandomRangeFloat(Minimum=0, Maximum=0.6)
    Module.Param0WriteEnabled=true
    Module.Param1WriteEnabled=true
    Module.Param2WriteEnabled=true
    Module.Param3WriteEnabled=true
   )
  ]
  particleUpdateScript [
   &ParticleState()
   /SetVariables(Particles.Velocity=&Multiply_VectorByFloat(
     Vector=$Particles.Initial.Velocity
     Float=&FloatFromCurve(FloatCurve=DataInterface() {propertyValues {Curve {keys [
          {arriveTangent=-5.5773, interpMode=RCIM_Cubic, leaveTangent=-5.5773, tangentMode=RCTM_User, time=0, value=1}
          {arriveTangent=-0.9831, interpMode=RCIM_Cubic, leaveTangent=-0.9831, tangentMode=RCTM_User, time=0.2018, value=0.1945}
          <at 0.75 0>
         ]}}})
    )) {isSetParameters=true, module=SetVariables_#3}
   &SolveForcesAndVelocity(Clamp Velocity=false, Limit Acceleration=false)
   &ScaleColor(
    Scale Alpha=&FloatFromCurve(FloatCurve=DataInterface() {propertyValues {Curve {keys [{interpMode=RCIM_Cubic, tangentMode=RCTM_User, time=0.75, value=1}, <at 1 0.5>]}}})
    Scale RGB=&Multiply_VectorByFloat(
     Vector=&MakeVectorFromLinearColorRGB(Color=&ColorFromCurve(ColorCurve=DataInterface() {propertyValues {
         AlphaCurve {keys []}
         BlueCurve {keys [<at 0.5 1>,<at 1 0 RCIM_Cubic>]}
         GreenCurve {keys [<at 0.5 1>,<at 1 0 RCIM_Cubic>]}
         RedCurve {keys [<at 0.5 1>,<at 1 0 RCIM_Cubic>]}
        }}))
     Float=1
    )
    ScaleA=true
    ScaleRGB=true
   ) {module=ScaleColor001}
   &ScaleSpriteSize(Scale Factor=&Vector2DFromFloat(Value=&FloatFromCurve(FloatCurve=DataInterface() {propertyValues {Curve {keys [
          {arriveTangent=11.6143, interpMode=RCIM_Cubic, leaveTangent=11.6143, tangentMode=RCTM_User, time=0, value=0.25}
          {arriveTangent=1.5, interpMode=RCIM_Cubic, leaveTangent=1.5, time=0.1, value=1.5}
          {interpMode=RCIM_Cubic, leaveTangent=0.8761, time=0.5, value=1}
         ]}}})))
  ]
  rendererClasses [&NiagaraSpriteRendererProperties]
  renderers [{rendererClass=&NiagaraSpriteRendererProperties}]
 }
}
summary {
 systemName=NS_Backpack_CloakedAssassin_NPC
 emitters [
  {emitterName=Lights, rendererClasses [&NiagaraLightRendererProperties]}
  {emitterName=Lightning, rendererClasses [&NiagaraSpriteRendererProperties]}
  {emitterName=Runes, rendererClasses [&NiagaraSpriteRendererProperties]}
  {emitterName=Lights_BackpackArms, rendererClasses [&NiagaraLightRendererProperties]}
  {emitterName=Polygon_Trail, rendererClasses [&NiagaraMeshRendererProperties]}
 ]
 userVariables [
  User.FXColor {type=LinearColor, description="", defaultValue {struct=LinearColor, value=<rgba 1 0 0.7667 1>}}
  User.SkeletalMesh {
   type=NiagaraDataInterfaceSkeletalMesh
   defaultValue {
    struct=NiagaraExt_VariableValue_DataInterface
    value {dataInterface=&self:NiagaraDataInterfaceSkeletalMesh_0}
   }
  }
 ]
}
